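# Niagara idle effect for the SkyRattle cosmetic shoes, adding lightning, glow lights and foot trails to both feet.
# Light attaches two point lights to skeletal bones, while ShoeLightning spawns short-lived mesh and sprite bolts there, gated by kill count and hidden flags.
# LightningTrail draws anim-trail ribbons only while running, with none in freefall.
# Per-foot shoe scale drives sizing, and time of day dims trail glow at night.
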NiagaraSystem /CosmeticShoes/Assets/FORT_Shoes/Shoe_SkyRattle/FX/NS_Shoe_Skyrattle_Idle
dependencies {
 usedDataInterfaces [
  &NiagaraDataInterfaceArrayFloat3
  &NiagaraDataInterfaceArrayFloat
  &NiagaraDataInterfaceCurve
  &NiagaraDataInterfaceColorCurve
  &NiagaraDataInterfaceSkeletalMesh
 ]
 usedDynamicInputs [
  /Game/Effects/Niagara/DynamicInputs/Logic/FN_IsRunningCheck
  &Boolean_AndOperation
  &VectorLength
  &Multiply_Vector
  &ModuloInt
  &Add_Integer
  /Niagara/DynamicInputs/Divide/Divide_Int
  &MakeIntFromFloat
 ]
 usedModules [
  &self:SystemSpawnScript.SystemScriptSource.SystemScriptGraph.NiagaraNodeAssignment_0.SetVariables_#1
  &SystemState
  &Timeline
  &self:SystemSpawnScript.SystemScriptSource.SystemScriptGraph.NiagaraNodeAssignment_1.SetVariables_#2
  &self:SystemSpawnScript.SystemScriptSource.SystemScriptGraph.NiagaraNodeAssignment_6.SetVariables_#3
  &self:SystemSpawnScript.SystemScriptSource.SystemScriptGraph.NiagaraNodeAssignment_5.SetVariables_#4
  &self:SystemSpawnScript.SystemScriptSource.SystemScriptGraph.NiagaraNodeAssignment_4.SetVariables_#5
 ]
}
emitters {
 Light {
  emitterUpdateScript [
   &EmitterState(
    Inactive Response=::NewEnumerator1
    Life Cycle Mode=::NewEnumerator1
    Loop Behavior=::NewEnumerator1
    Loop Duration Mode=::NewEnumerator1
    UseLoopDelay=false
   )
   &SpawnBurst_Instantaneous(Spawn Count=2, Use Spawn Probability=false, UseLoopCountLimit=false)
   /SetVariables(Emitter.LightActive:NiagaraBool=&Boolean_OrOperation(
     Bool1=&SetBoolByIntComparison(A=$System.KillCountInt, B=4, Comparison Type=::NewEnumerator2)
     Bool2=&Boolean_OrOperation(Bool1=&Boolean_OrOperation(Bool1=$User.bIsShoeHidden, Bool2=$User.bIsFootHidden), Bool2=$System.isFreeFalling)
    )) {isSetParameters=true}
  ]
  particleSpawnScript [
   &PartitionParticles()
   &InitializeParticle(Color=<rgba 11 6.977 0.3163 1>, UsePositionOffset=false, Write Material Random=true)
  ]
  particleUpdateScript [
   &ParticleState(Kill Particles When Lifetime Has Elapsed=false, Let Infinitely Lived Particles Die When Emitter Deactivates=true)
   &SkeletalMeshLocation(
    Bone / Socket Index=$Particles.PartitionParticles.Partition
    Bone Sampling Mode=::NewEnumerator4
    MaskSpawnGroup=false
    Sampled Position Offset=&SelectVectorFromArray(
     Direct Array Index=$Particles.PartitionParticles.Partition
     Vector Selection Array=DataInterface() {propertyValues {FloatData [<-4 3 0>, <4 -3 0>]}}
    )
    Skeletal Mesh=$System.Skeletal Mesh Reader
   )
   &FN_Light_Attributes(
    Color=$Particles.Color
    Create Color Attribute=true
    Create Exponent Attribute=true
    Create Radius Attribute=true
    Enable Particle Light Enable=true
    Exponent=3
    Light Enabled:NiagaraBool=&InvertBool(Boolean:NiagaraBool=&Boolean_OrOperation(
      Bool1=&SetBoolByFloatComparison(
       A=&MakeFloatFromVector(
        Vector=&VectorFromFloat(Value=&MaskFloatBySpawnGroup(
          Masked Float=&MakeFloatFromVector(Vector=$User.ShoeScale_L, Channel=::NewEnumerator0)
          Passthrough Float=&MakeFloatFromVector(Vector=$User.ShoeScale_R, Channel=::NewEnumerator0)
          Spawn Group Mask=$Particles.PartitionParticles.Partition
         ))
        Channel=::NewEnumerator0
       )
       B=0
       Comparison Type=::NewEnumerator4
      )
      Bool2=&Boolean_OrOperation(Bool1=$User.bIsShoeHidden, Bool2=$User.bIsFootHidden)
     ))
    Radius=22
    Scale Color=$System.Timeline.Float
   )
  ]
  rendererClasses [&NiagaraLightRendererProperties]
  renderers [{rendererClass=&NiagaraLightRendererProperties, rendererIndex=0}]
 }
 LightningTrail {
  emitterUpdateScript [
   &EmitterState()
   &FN_SpawnAnimTrail(
    AnimTrailWidthMode=::FromFirst
    Number of Anim Trails:NiagaraInt32=&MakeCustomIntFromBool(Bool=$System.isFreeFalling, True INT=0, False INT=2)
    Spawn Mode:ENiagaraSkeletalSpawning=::NewEnumerator0
    Switch on ENiagara_AnimTrail_Source:ENiagara_AnimTrail_Source=::NewEnumerator1
   )
  ]
  particleSpawnScript [
   &PartitionParticles()
   &FN_InitializeAnimTrailBoneIndecies(Second Bone or Socket Index=0)
   &FN_InitializeAnimTrailBoneIndecies(AnimTrail Selection=1, First Bone or Socket Index=1) {module=FN_InitializeAnimTrailBoneIndecies001}
   &FN_InitializeAnimTrail(AnimTrailWidth=11, Lifetime=0.3, SkeletalMesh=$System.Skeletal Mesh Reader)
  ]
  particleUpdateScript [
   &ParticleState()
   &FN_UpdateAnimTrail(SkeletalMesh=$System.Skeletal Mesh Reader)
   /SetVariables(Particles.RibbonWidth=&Multiply_Float(
     A=$Particles.RibbonWidth
     B=&MaskFloatBySpawnGroup(
      Masked Float=&MakeFloatFromVector(Vector=$User.ShoeScale_L, Channel=::NewEnumerator0)
      Passthrough Float=&MakeFloatFromVector(Vector=$User.ShoeScale_R, Channel=::NewEnumerator0)
      Spawn Group Mask=$Emitter.SpawnGroup
     )
    )) {isSetParameters=true}
   &KillParticles(Kill Particles=&Boolean_OrOperation(Bool1=&InvertBool(Boolean:NiagaraBool=$System.isRunning), Bool2=$User.HideFootFX))
   &DynamicMaterialParameters(
    Index 0 Param 1=2
    Index 0 Param 2=&FloatFromCurve(FloatCurve=DataInterface() {propertyValues {Curve {keys [
         {arriveTangent=-0.1402, interpMode=RCIM_Cubic, leaveTangent=-0.1402, tangentMode=RCTM_User, time=0, value=0.05}
         {interpMode=RCIM_Cubic, tangentMode=RCTM_User, time=1, value=0}
        ]}}})
    Index 0 Param 3=10
    Module.Index0Float0Write=true
    Module.Index0Float1Write=true
    Module.Index0Float2Write=true
    Module.Index0Float3Write=true
    Module.Index1Float0Write=true
    Module.Index1Float1Write=true
    Module.Index1Float2Write=true
    Module.Index1Float3Write=true
    Write Parameter Index 1=::NewEnumerator1
   )
   &ScaleColor(
    Scale Alpha=&FloatFromCurve(
     FloatCurve=DataInterface() {propertyValues {Curve {keys [
         {arriveTangent=-7.1431, interpMode=RCIM_Cubic, leaveTangent=-7.1431, tangentMode=RCTM_User, time=0.2, value=1}
         {arriveTangent=0.1174, tangentMode=RCTM_User, time=0.6, value=0}
        ]}}}
     Scale Curve=&Multiply_Float(A=$System.Timeline001.Float, B=$System.Timeline002.Float)
    )
    Scale RGB=&Multiply_VectorByFloat(
     Vector=&MakeVectorFromLinearColorRGB(Color=&ColorFromCurve(ColorCurve=DataInterface() {propertyValues {
         AlphaCurve {keys []}
         BlueCurve {keys [<at 0.1797 0.2371 RCIM_Cubic>,<at 0.9993 0.2141 RCIM_Cubic>]}
         GreenCurve {keys [<at 0.1797 2.1471 RCIM_Cubic>,<at 0.9993 0.6148 RCIM_Cubic>]}
         RedCurve {keys [<at 0.1797 3 RCIM_Cubic>,<at 0.9993 1 RCIM_Cubic>]}
        }}))
     Float=&Multiply_Float(A=1, B=&FNTimeOfDayScalar(Night Scale=0.8, Morning Scale=1, Evening Scale=1))
    )
    ScaleA=true
    ScaleRGB=true
   )
   /SetVariables(
    Particles.GlowColor=&MakeLinearColorFromVectorAndFloat(
     "Vector (RGB)"=&MakeVectorFromLinearColorRGB(Color=&MultiplyLinearColor(
       Linear Color=<rgba 1 0.6981 0.0084 1>
       Scale Factor=&MakeVector4_FromLinearColor(LinearColor=&ColorFromCurve(ColorCurve=DataInterface() {propertyValues {
           AlphaCurve {keys []}
           BlueCurve {keys [<at 0.4593 1 RCIM_Cubic>,<at 1 0.7>]}
           GreenCurve {keys [<at 0.4593 1 RCIM_Cubic>,<at 1 0.7>]}
           RedCurve {keys [<at 0.4593 1 RCIM_Cubic>,<at 1 0.7>]}
          }}))
      ))
     "Float (Alpha)"=&Multiply_Float(
      A=&FNTimeOfDayScalar(Night Scale=0.5, Morning Scale=1, Evening Scale=1)
      B=&FloatFromCurve(
       FloatCurve=DataInterface() {propertyValues {Curve {keys [
           {arriveTangent=-7.1431, interpMode=RCIM_Cubic, leaveTangent=-7.1431, tangentMode=RCTM_User, time=0.2, value=1}
           {arriveTangent=0.1174, tangentMode=RCTM_User, time=0.6, value=0}
          ]}}}
       Scale Curve=&Multiply_Float(A=$System.Timeline001.Float, B=$System.Timeline002.Float)
      )
     )
    )
    Particles.GlowDyn2:Vector4f=<1 1 1 1>
    Particles.GlowWidth:NiagaraFloat=&Multiply_Float(A=$Particles.RibbonWidth, B=1.4)
   ) {isSetParameters=true}
  ]
  rendererClasses [&NiagaraRibbonRendererProperties]
  renderers [
   {rendererClass=&NiagaraRibbonRendererProperties}
   {rendererClass=&NiagaraRibbonRendererProperties, rendererIndex=1}
  ]
 }
 ShoeLightning {
  emitterUpdateScript [
   &EmitterState()
   &SpawnRate(
    Spawn Probability=&MakeCustomFloatFromBool(
     Bool=&Boolean_OrOperation(
      Bool1=&SetBoolByIntComparison(A=$System.KillCountInt, B=3, Comparison Type=::NewEnumerator2)
      Bool2=&Boolean_OrOperation(Bool1=$User.bIsShoeHidden, Bool2=$User.bIsFootHidden)
     )
     True Float=0.22
     False Float=0
    )
    SpawnRate=22
    Use Spawn Probability=true
   )
  ]
  particleSpawnScript [
   &PartitionParticles()
   &InitializeParticle(
    Color=&MultiplyLinearColor(
     Linear Color=&RandomRangeLinearColor(Minimum=<rgba 1 1 1 0.3>, Maximum=<rgba 1.2 1.2 1.2 1.5>)
     Scale Factor=<1 1 1 1>
    )
    Lifetime=&RandomRangeFloat(Minimum=0.15, Maximum=0.2)
    Mesh Scale Mode=::NewEnumerator3
    Mesh Uniform Scale=&Multiply_Float(A=&RandomRangeFloat(Minimum=0.03, Maximum=0.05), B=1)
    Sprite Size Mode=::NewEnumerator3
    Uniform Sprite Size=15
    UsePositionOffset=false
    Write Material Random=true
   )
   &CameraOffset(Camera Offset Amount=5)
   &ShapeLocation(
    Box / Plane Mode:ENiagara_BoxPlaneMode=::NewEnumerator1
    MaskSpawnGroup=false
    Offset=&MaskVectorBySpawnGroup(Masked Vector=<0 2 -7>, Passthrough Vector=<0 -2 7>, Spawn Group Mask=$Particles.PartitionParticles.Partition)
    Offset Coordinate Space=::Local
    Offset Mode=::NewEnumerator0
    Override Seed=false
    Plane Size:Vector2f=&Multiply_Vector2D_ByFloat(Float=$System.ShoeScale, Vector2D=<9 22>)
    Rotation Axis=<0 1 0>
    Rotation Mode=::NewEnumerator1
    Shape Primitive=::NewEnumerator2
    UseOwnerScale=true
   )
   &InitialMeshOrientation(Mesh Axis To Orient=<0 0 1>, Mesh Orientation Mode=::NewEnumerator2, Use Rotation Vector=false)
   /SetVariables() {isSetParameters=true}
  ]
  particleUpdateScript [
   &ParticleState()
   &SkeletalMeshLocation(
    Bone / Socket Index=$Particles.PartitionParticles.Partition
    Bone Sampling Mode=::NewEnumerator4
    MaskSpawnGroup=false
    Sampled Position Offset:NiagaraPosition=$Particles.ShapeLocation.ShapePosition
    Sampled Position Offset Coordinate Space:ENiagaraSkelMeshTransforms=::NewEnumerator4
    Skeletal Mesh=$System.Skeletal Mesh Reader
   ) {module=SkeletalMeshLocation001}
   &ScaleMeshSize(Scale Factor=&Multiply_VectorByFloat(
     Vector=&RandomRangeVector(Minimum=<1 1 2.2>, Maximum=<1.3 1.3 2.2>)
     Float=&MakeFloatFromVector(
      Vector=&VectorFromFloat(Value=&MaskFloatBySpawnGroup(
        Masked Float=&MakeFloatFromVector(Vector=$User.ShoeScale_L, Channel=::NewEnumerator0)
        Passthrough Float=&MakeFloatFromVector(Vector=$User.ShoeScale_R, Channel=::NewEnumerator0)
        Spawn Group Mask=$Particles.PartitionParticles.Partition
       ))
      Channel=::NewEnumerator0
     )
    )) {module=ScaleMeshSize001}
   &DynamicMaterialParameters(
    Index 0 Param 2=&RandomRangeFloat(Minimum=0, Maximum=2)
    Index 0 Param 4=&MakeCustomFloatFromBool(
     Bool=false
     True Float=11
     False Float=&SelectFloatFromArray(
      Array Sampling Mode=::NewEnumerator0
      Direct Array Index=?
      Float Selection Array=DataInterface() {propertyValues {FloatData [3,11,1,9,2,11,11]}}
      Random Seed=0
     )
    )
    Index 1 Param 1=&FloatFromCurve(FloatCurve=DataInterface() {propertyValues {Curve {keys [<at 0 1>,<at 1 1.3>]}}})
    Index 1 Param 2=2
    Index 1 Param 3=&FloatFromCurve(FloatCurve=DataInterface() {propertyValues {Curve {keys [{interpMode=RCIM_Cubic, tangentMode=RCTM_Break, time=0.5, value=-0.5}, <at 1 0.1>]}}})
    Index 2 Param 1=&Multiply_Float(
     A=&FloatFromCurve(
      FloatCurve=DataInterface() {propertyValues {Curve {keys [{interpMode=RCIM_Cubic, leaveTangent=8.1085, tangentMode=RCTM_Break, time=0.8568, value=1}, {tangentMode=RCTM_User, time=1, value=0}]}}}
      Scale Curve=0.5
     )
     B=&Divide_Float(
      A=&MakeFloatFromVector(Vector=$Particles.Scale, Channel=::NewEnumerator0)
      B=&RandomRangeFloat(Minimum=7, Maximum=10)
     )
    )
    Index 2 Param 2=&FloatFromCurve(
     FloatCurve=DataInterface() {propertyValues {Curve {keys [
         {interpMode=RCIM_Cubic, tangentMode=RCTM_Break, time=0.7, value=0}
         {arriveTangent=-0.0859, interpMode=RCIM_Cubic, tangentMode=RCTM_Break, time=1, value=-1}
        ]}}}
     Scale Curve=50
    )
    Index 2 Param 3=&Waveform(
     Global Input Position=&RandomRangeFloat(Minimum=0, Maximum=15)
     Global Phase=$Particles.Age
     "Waveform [1]"=::NewEnumerator12
     "[1] Phase Scale"=2
     "[1] Frequency"=1
     "[1] Amplitude Min/Max"=<0 1>
    )
    Index 2 Param 4=25
    Index 3 Param 1=&FloatFromCurve(
     FloatCurve=DataInterface() {propertyValues {Curve {keys [
         {interpMode=RCIM_Cubic, leaveTangent=-0.0025, tangentMode=RCTM_Break, time=0, value=0}
         {arriveTangent=1.8528, interpMode=RCIM_Cubic, tangentMode=RCTM_Break, time=1, value=1}
        ]}}}
     Scale Curve=&RandomRangeFloat(Minimum=15, Maximum=25)
    )
    Index 3 Param 2=&FloatFromCurve(FloatCurve=DataInterface() {propertyValues {Curve {keys [
         {interpMode=RCIM_Cubic, leaveTangent=-0.0025, tangentMode=RCTM_Break, time=0, value=0.5}
         {arriveTangent=-0.0316, interpMode=RCIM_Cubic, tangentMode=RCTM_Break, time=1, value=-0.2}
        ]}}})
    Index 3 Param 3=&FloatFromCurve(FloatCurve=DataInterface() {propertyValues {Curve {keys [
         {interpMode=RCIM_Cubic, leaveTangent=1.1034, tangentMode=RCTM_Break, time=0, value=-1}
         {arriveTangent=2.2399, tangentMode=RCTM_User, time=0.5, value=0.125}
        ]}}})
    Index 3 Param 4=&FloatFromCurve(FloatCurve=DataInterface() {propertyValues {Curve {keys [
         {interpMode=RCIM_Cubic, leaveTangent=3.0601, tangentMode=RCTM_Break, time=0.5, value=-0.5}
         {arriveTangent=3.2053, tangentMode=RCTM_User, time=1, value=1}
        ]}}})
    Module.Index0Param0Write=true
    Module.Index0Param1Write=true
    Module.Index0Param2Write=true
    Module.Index0Param3Write=true
    Module.Index1Param0Write=true
    Module.Index1Param1Write=true
    Module.Index1Param2Write=true
    Module.Index1Param3Write=true
    Module.Index2Param0Write=true
    Module.Index2Param1Write=true
    Module.Index2Param2Write=true
    Module.Index2Param3Write=true
    Module.Index3Param0Write=true
    Module.Index3Param1Write=true
    Module.Index3Param2Write=true
    Module.Index3Param3Write=true
    Write Parameter Index 1=true
    Write Parameter Index 2=true
    Write Parameter Index 3=true
   )
   &ScaleColor(
    Scale Alpha=&FloatFromCurve(FloatCurve=DataInterface() {propertyValues {Curve {keys [
         {tangentMode=RCTM_User, time=0, value=1}
         {arriveTangent=-0.0405, interpMode=RCIM_Cubic, leaveTangent=-0.0405, tangentMode=RCTM_User, time=0.7009, value=1}
         {arriveTangent=-8.0089, interpMode=RCIM_Cubic, leaveTangent=-8.0089, tangentMode=RCTM_User, time=1, value=0}
        ]}}})
    Scale RGB=&Lerp_Vector(
     A=&VectorFromFloat(Value=&FloatFromCurve(FloatCurve=DataInterface() {propertyValues {Curve {keys [<at 0.2771 1>,<at 0.3786 1>]}}}, Scale Curve=20))
     B=&VectorFromFloat(Value=&RandomRangeFloat(Minimum=0.3, Maximum=2, Evaluation Type=::NewEnumerator1))
     Alpha=&FloatFromCurve(FloatCurve=DataInterface() {propertyValues {Curve {keys [
          {interpMode=RCIM_Cubic, leaveTangent=2.5294, tangentMode=RCTM_Break, time=0, value=0}
          {arriveTangent=-0.2166, interpMode=RCIM_Cubic, tangentMode=RCTM_Break, time=0.4, value=1}
         ]}}})
    )
    ScaleA=true
    ScaleRGB=true
   )
   &ScaleColor(ScaleA=true, ScaleRGB=true) {module=ScaleColor001}
   /SetVariables(Particles.MeshOrientation=&OrientMeshToVector(
     Look At Direction=&NormalizeVector(
      Vector To Normalize=&SubtractVectorFromPosition(
       Position=$Particles.Position
       Vector=&AddVectorToPosition(
        Position=$Particles.SkeletalMeshLocation001.SampledPosition
        Vector=&TransformVector(Vector=<0 0 -5>, Destination Space=::Simulation)
       )
      )
      Safe Normalize=true
      Safe Fallback Threshold=0
      Safe Fallback Vector=<0 0 0.001>
     )
     Reference Axis Vector=<1 0 0>
    )) {isSetParameters=true}
   /SetVariables(
    Particles.Fakelightglow:LinearColor=&MakeLinearColorFromVectorAndFloat(
     "Vector (RGB)"=&MakeVectorFromLinearColorRGB(Color=&MultiplyLinearColor(Linear Color=$Particles.Color, Scale Factor=&MakeVector4_FromLinearColor(LinearColor=<rgba 1 0.7336 0 1>)))
     "Float (Alpha)"=0.5
    )
    Particles.GlowSplineColor:LinearColor=&Multiply_LinearColorByFloat(Linear Color=$Particles.Color, Float=0.25)
   ) {isSetParameters=true}
  ]
  rendererClasses [&NiagaraMeshRendererProperties, &NiagaraSpriteRendererProperties]
  renderers [
   {rendererClass=&NiagaraMeshRendererProperties, rendererIndex=0}
   {rendererClass=&NiagaraMeshRendererProperties, rendererIndex=1}
   {rendererClass=&NiagaraSpriteRendererProperties, rendererIndex=2}
  ]
 }
}
summary {
 systemName=NS_Shoe_Skyrattle_Idle
 emitters [
  {emitterName=Light, rendererClasses [&NiagaraLightRendererProperties]}
  {emitterName=ShoeLightning, rendererClasses [&NiagaraMeshRendererProperties, &NiagaraSpriteRendererProperties]}
  {emitterName=LightningTrail, rendererClasses [&NiagaraRibbonRendererProperties]}
 ]
 userVariables [
  User.bIsFootHidden {type=NiagaraBool}
  User.bIsFrontEnd {type=NiagaraBool}
  User.bIsFrontEndPreview {type=NiagaraBool}
  User.bIsShoeHidden {type=NiagaraBool}
  User.HideFootFX {type=NiagaraBool}
  User.isOffPawnPreview {type=NiagaraBool}
  User.IsSkydiving {type=NiagaraBool}
  User.KillCount {type=NiagaraFloat, defaultValue=0, description=""}
  User.KillCountInt {type=NiagaraInt32}
  User.RibbonIDOffset {type=NiagaraInt32}
  User.ShoeScale_L {type=Vector3f, description="", defaultValue {struct=Vector3f, value=<1 1 1>}}
  User.ShoeScale_R {type=Vector3f, description="", defaultValue {struct=Vector3f, value=<1 1 1>}}
 ]
}
