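# Material function for the SkyRattle cosmetic shoe that transforms UVs, exposed to the library as "Transform (UV)" under Motion Design and Material Designer.
# It takes texture coordinates, scale, rotation, rotation center and offset inputs, routed through named reroutes.
# A subtract and multiply stage feeds CustomRotator with a 0-1 angle, then two additions produce the output.
# A Multiply on the rotation value is left unconnected.
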MaterialFunction /CosmeticShoes/Assets/FORT_Shoes/Shoe_SkyRattle/Materials/MF_SkyRattle_Transform
graph {
 properties {bExposeToLibrary=true, userExposedCaption="Transform (UV)", libraryCategoriesText [Motion Design, Material Designer]}
 nodes [
  n1 = NamedRerouteUsage()
  n2 = NamedRerouteUsage()
  n3 = NamedRerouteUsage()
  n4 = MaterialFunctionCall(UVs=Multiply(A=Subtract(A=@n1, B=@n2), B=NamedRerouteUsage()), Rotation Center=NamedRerouteUsage(), "Rotation Angle (0-1)"=@n3) {materialFunction=&CustomRotator}
  n5 = NamedRerouteUsage()
  FunctionOutput(None=Add(A=Add(A=@n4.Rotated Values, B=@n5), B=NamedRerouteUsage()))
  Multiply(A=@n3, B=NamedRerouteUsage())
  NamedRerouteDeclaration(None=FunctionInput() {inputName=In_Scale, inputType=FunctionInput_Vector2, sortPriority=1}) {name=In_Scale, nodeColor=<rgba 0 0.8828 1 1>}
  NamedRerouteDeclaration(None=FunctionInput() {inputName=In_TextureCoordinates, inputType=FunctionInput_Vector2}) {name=TextureCoordinates, nodeColor=<rgba 1 0.7969 0 1>}
  NamedRerouteDeclaration(None=FunctionInput() {inputName=In_Rotation, inputType=FunctionInput_Scalar, sortPriority=2}) {name=In_Rotation, nodeColor=<rgba 0.4766 0 1 1>}
  NamedRerouteDeclaration(None=FunctionInput() {inputName=In_RotationCenter, inputType=FunctionInput_Vector2, sortPriority=3}) {name=In_RotationCenter, nodeColor=<rgba 0 1 0.1562 1>}
  NamedRerouteDeclaration(None=FunctionInput() {inputName=In_Offset, inputType=FunctionInput_Vector2, sortPriority=4}) {name=In_Offset, nodeColor=<rgba 1 0 0.1641 1>}
 ]
}
