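# SkyRattle shoe material function that adds animated noise and glow to base color and emissive, driven by a KillCount parameter.
# Red, green and blue noise channels scroll with per-region speed, scale, rotation and pivot, with warped UVs and three-color gradients.
# KillCount, capped and blendable for frontend display, feeds reveal masks and emissive boost.
# Masks shape sole glow, sharp edges and sine-pulsed blurred edges.
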MaterialFunction /CosmeticShoes/Assets/FORT_Shoes/Shoe_SkyRattle/Materials/MF_SkyRattle_TechArt
graph {nodes [
  n1 = NamedRerouteUsage()
  Divide(A=@n1, B=NamedRerouteUsage())
  n2 = GetMaterialAttributes(None=NamedRerouteUsage()) {attributeGetTypes [#1, #2]}
  FunctionOutput(None=SetMaterialAttributes!(MaterialAttributes=@n2.MaterialAttributes, BaseColor=NamedRerouteUsage(), EmissiveColor=NamedRerouteUsage()) {attributeSetTypes [#1, #2]}) {outputName=MaterialAttributes}
  NamedRerouteDeclaration(None=FunctionInput() {inputName=In_MaterialAttributes, inputType=FunctionInput_MaterialAttributes}) {name=In_MaterialAttributes, nodeColor=<rgba 0.0312 1 0 1>}
  n3 = TextureSampleParameter2D(UVs=TextureCoordinate()) {
   parameterName=TechArtMaskDefault
   samplerType=SAMPLERTYPE_Masks
   texture=/CosmeticShoes/Assets/FORT_Shoes/Shoe_SkyRattle/Textures/T_SkyRattle_Shoe_FX
  }
  NamedRerouteDeclaration(None=FunctionInput(Preview=@n3.RGB) {bUsePreviewValueAsDefault=true, inputName=In_TechArtMask, sortPriority=1}) {name=In_TechArtMask, nodeColor=<rgba 1 0 0.3516 1>}
  n4 = AppendVector(
   A=ScalarParameter() {group=SkyRattle_TechArt, parameterName=BlueChannelSpeedX_Shoes}
   B=ScalarParameter() {defaultValue=0.1, group=SkyRattle_TechArt, parameterName=BlueChannelSpeedY_Shoes}
  )
  n5 = AppendVector(
   A=ScalarParameter() {group=SkyRattle_TechArt, parameterName=BlueChannelSpeedX_SoleOfShoes}
   B=ScalarParameter() {defaultValue=0.1, group=SkyRattle_TechArt, parameterName=BlueChannelSpeedY_SoleOfShoes}
  )
  n6 = ComponentMask(None=NamedRerouteUsage()) {b=true}
  NamedRerouteDeclaration(None=Multiply(A=Time(), B=LinearInterpolate(A=@n4, B=@n5, Alpha=@n6) {constB=0})) {name=BlueChannelUVLerpAlpha, nodeColor=<rgba 0.0042 0.0244 1 1>}
  n7 = TextureCoordinate()
  n8 = NamedRerouteUsage()
  n9 = TextureSampleParameter2D(UVs=MaterialFunctionCall!(
    In_TextureCoordinates=Add(A=@n7, B=@n8)
    In_Scale=NamedRerouteUsage()
    In_Rotation=@n1
    In_RotationCenter=NamedRerouteUsage()
    In_Offset=NamedRerouteUsage()
   ) {materialFunction=&/MF_SkyRattle_Transform}.Result) {
   parameterName=NoiseFlames
   samplerSource=SSM_Wrap_WorldGroupSettings
   samplerType=SAMPLERTYPE_LinearColor
   texture=/CosmeticShoes/Assets/FORT_Shoes/Shoe_SkyRattle/Textures/T_SkyRattle_Shoe_Noise
  }
  NamedRerouteDeclaration(None=@n9.R) {name=RedChannelAnimated, nodeColor=<rgba 1 0.0004 0 1>}
  n10 = TextureCoordinate()
  n11 = NamedRerouteUsage()
  n12 = NamedRerouteUsage()
  n13 = NamedRerouteUsage()
  n14 = TextureSampleParameter2D(UVs=MaterialFunctionCall(
    In_TextureCoordinates=Add(A=@n10, B=@n11)
    In_Scale=NamedRerouteUsage()
    In_Rotation=Divide(A=@n12, B=@n13)
    In_RotationCenter=NamedRerouteUsage()
    In_Offset=NamedRerouteUsage()
   ) {materialFunction=&MF_SkyRattle_Transform}.Result) {
   parameterName=NoiseFlames
   samplerSource=SSM_Wrap_WorldGroupSettings
   samplerType=SAMPLERTYPE_LinearColor
   texture=&T_SkyRattle_Shoe_Noise
  }
  NamedRerouteDeclaration(None=@n14.G) {name=GreenChannelAnimated, nodeColor=<rgba 0.0713 1 0 1>}
  n15 = TextureCoordinate()
  n16 = NamedRerouteUsage()
  n17 = NamedRerouteUsage()
  n18 = NamedRerouteUsage()
  n19 = TextureSampleParameter2D(UVs=MaterialFunctionCall(
    In_TextureCoordinates=Add(A=@n15, B=@n16)
    In_Scale=NamedRerouteUsage()
    In_Rotation=Divide(A=@n17, B=@n18)
    In_RotationCenter=NamedRerouteUsage()
    In_Offset=NamedRerouteUsage()
   ) {materialFunction=&MF_SkyRattle_Transform}.Result) {
   parameterName=NoiseFlames
   samplerSource=SSM_Wrap_WorldGroupSettings
   samplerType=SAMPLERTYPE_LinearColor
   texture=&T_SkyRattle_Shoe_Noise
  }
  NamedRerouteDeclaration(None=@n19.B) {name=BlueChannelAnimated, nodeColor=<rgba 0 0.1094 1 1>}
  NamedRerouteDeclaration(None=MaterialFunctionCall(
    A=VectorParameter() {group=SkyRattle_TechArt, parameterName=GreenChannelColorA, defaultValue=<rgba 0 0.02 0 1>}.RGB
    B=VectorParameter() {group=SkyRattle_TechArt, parameterName=GreenChannelColorB, defaultValue=<rgba 0.0093 0.2696 0 1>}.RGB
    C=VectorParameter() {group=SkyRattle_TechArt, parameterName=GreenChannelColorC, defaultValue=<rgba 0.0271 0.85 0 1>}.RGB
    Alpha=NamedRerouteUsage()
   ) {materialFunction=&Lerp_3Color}.Result) {name=GreenChannelColor, nodeColor=<rgba 0 1 0.0514 1>}
  NamedRerouteDeclaration(None=MaterialFunctionCall(
    A=VectorParameter() {group=SkyRattle_TechArt, parameterName=BlueChannelColorA, defaultValue=<rgba 0.0001 0 0.02 1>}.RGB
    B=VectorParameter() {group=SkyRattle_TechArt, parameterName=BlueChannelColorB, defaultValue=<rgba 0.0089 0.0004 0.2696 1>}.RGB
    C=VectorParameter() {group=SkyRattle_TechArt, parameterName=BlueChannelColorC, defaultValue=<rgba 0.0017 0.0831 0.85 1>}.RGB
    Alpha=NamedRerouteUsage()
   ) {materialFunction=&Lerp_3Color}.Result) {name=BlueChannelColor, nodeColor=<rgba 0 0.0055 1 1>}
  n20 = NamedRerouteUsage()
  n21 = NamedRerouteUsage()
  NamedRerouteDeclaration(None=Multiply(A=Multiply(A=@n20, B=@n21), B=NamedRerouteUsage())) {name=NoiseColor, nodeColor=<rgba 1 0 0.4922 1>}
  n22 = TextureSampleParameter2D(UVs=MaterialFunctionCall(
    In_TextureCoordinates=TextureCoordinate()
    In_Scale=NamedRerouteUsage()
    In_Rotation=NamedRerouteUsage()
    In_RotationCenter=NamedRerouteUsage()
    In_Offset=NamedRerouteUsage()
   ) {materialFunction=&MF_SkyRattle_Transform}.Result) {
   parameterName=NoiseFlames
   samplerSource=SSM_Wrap_WorldGroupSettings
   samplerType=SAMPLERTYPE_LinearColor
   texture=&T_SkyRattle_Shoe_Noise
  }
  NamedRerouteDeclaration(None=@n22.R) {name=RedUVWarp, nodeColor=<rgba 1 0.0004 0 1>}
  n23 = TextureSampleParameter2D(UVs=MaterialFunctionCall(
    In_TextureCoordinates=TextureCoordinate()
    In_Scale=NamedRerouteUsage()
    In_Rotation=NamedRerouteUsage()
    In_RotationCenter=NamedRerouteUsage()
    In_Offset=NamedRerouteUsage()
   ) {materialFunction=&MF_SkyRattle_Transform}.Result) {
   parameterName=NoiseFlames
   samplerSource=SSM_Wrap_WorldGroupSettings
   samplerType=SAMPLERTYPE_LinearColor
   texture=&T_SkyRattle_Shoe_Noise
  }
  NamedRerouteDeclaration(None=@n23.G) {name=GreenUVWarp, nodeColor=<rgba 0.0713 1 0 1>}
  n24 = NamedRerouteUsage()
  n25 = NamedRerouteUsage()
  NamedRerouteDeclaration(None=Multiply(A=AppendVector(A=@n24, B=@n25), B=ScalarParameter() {defaultValue=0.1, parameterName=UVWarpMultiplier})) {name=UVWarp, nodeColor=<rgba 0 0.2031 1 1>}
  n26 = NamedRerouteUsage()
  n27 = Add(A=Subtract(A=Saturate(None=@n26)), B=Divide(A=NamedRerouteUsage()) {constB=4})
  NamedRerouteDeclaration(None=@n27) {name=KillCountMask, nodeColor=<rgba 0 0.6719 1 1>}
  n28 = NamedRerouteUsage()
  n29 = ScalarParameter() {defaultValue=1.1, parameterName=FourthLevelBoost}
  n30 = NamedRerouteUsage()
  n31 = ScalarParameter() {defaultValue=0.25, parameterName=ThirdLevelBoost}
  NamedRerouteDeclaration(None=If(A=NamedRerouteUsage(), "A > B"=Multiply(A=@n28, B=@n29), "A == B"=Multiply(A=@n30, B=@n31), "A < B"=Constant() {r=2}) {constB=3}) {name=EmissiveBoostPerElimination, nodeColor=<rgba 1 0 0.3281 1>}
  n32 = ScalarParameter() {defaultValue=2, parameterName=KillCount}
  n33 = Constant() {r=4}
  NamedRerouteDeclaration(None=LinearInterpolate(
    A=If(A=@n32, B=@n33, "A > B"=@n33, "A == B"=@n32, "A < B"=@n32) {constB=3}
    B=NamedRerouteUsage()
    Alpha=ScalarParameter() {group=ScorchPoint, parameterName=IsShoeInFrontend}
   )) {name=KillCountAmountCapped, nodeColor=<rgba 1 0.2578 0 1>}
  n34 = TextureSampleParameter2D(UVs=TextureCoordinate()) {
   parameterName=TechArtMask2
   samplerType=SAMPLERTYPE_LinearColor
   texture=/CosmeticShoes/Assets/FORT_Shoes/Shoe_SkyRattle/Textures/T_SkyRattle_Shoe_FX2
  }
  n35 = Add(A=@n34.R, B=@n34.R)
  NamedRerouteDeclaration(None=Saturate(None=Add(A=@n35, B=@n34.B))) {name=GradientMask, nodeColor=<rgba 0 0.0625 1 1>}
  NamedRerouteDeclaration(None=@n34.RGB) {name=BlurredMaskRGB, nodeColor=<rgba 0 1 0.3438 1>}
  n36 = Time()
  NamedRerouteDeclaration(None=Fmod(A=Divide(A=@n36) {constB=4, desc=4 times slower}, B=Constant() {r=6}) {desc=0 -6 loop}) {name=TimeLoopForOffPawnFE, nodeColor=<rgba 1 0.3516 0 1>}
  n37 = Multiply(
   A=VectorParameter() {group=SkyRattle_TechArt, parameterName=GlowColor, defaultValue=<rgba 0.7565 0.5766 0.1974 1>}.RGB
   B=ComponentMask(None=NamedRerouteUsage()) {b=true}
  )
  n38 = NamedRerouteUsage()
  NamedRerouteDeclaration(None=Multiply(A=Multiply(A=@n37, B=@n38), B=ScalarParameter() {defaultValue=0.25, parameterName=SoleOfShoeGlowMultiplier})) {name=SoleOfShoeGlow, nodeColor=<rgba 1 0 0.5859 1>}
  n39 = NamedRerouteUsage()
  n40 = Add(A=Add(A=ComponentMask(None=@n39) {r=true}, B=ComponentMask(None=@n39) {g=true}), B=ComponentMask(None=@n39) {b=true})
  NamedRerouteDeclaration(None=Subtract(A=NamedRerouteUsage(), B=Saturate(None=@n40))) {name=BlurredEdges, nodeColor=<rgba 0.2656 1 0 1>}
  NamedRerouteDeclaration(None=Multiply(
    A=VectorParameter() {group=SkyRattle_TechArt, parameterName=GlowColor, defaultValue=<rgba 0.7565 0.5766 0.1974 1>}.RGB
    B=NamedRerouteUsage()
   )) {name=BlurredEdgesColor, nodeColor=<rgba 1 0 0.5859 1>}
  n41 = NamedRerouteUsage()
  n42 = Add(A=Subtract(A=Saturate(None=@n41)), B=Divide(A=NamedRerouteUsage()) {constB=4})
  NamedRerouteDeclaration(None=@n42) {name=BlurredEdgesMask, nodeColor=<rgba 1 0.7734 0 1>}
  n43 = TextureSampleParameter2D(UVs=TextureCoordinate()) {
   parameterName=RotationMask
   samplerType=SAMPLERTYPE_LinearColor
   texture=/CosmeticShoes/Assets/FORT_Shoes/Shoe_SkyRattle/Textures/T_SkyRattle_Shoe_RotationMask
  }
  NamedRerouteDeclaration(None=@n43.RGB) {name=RotationMask, nodeColor=<rgba 1 0 0.5859 1>}
  NamedRerouteDeclaration(None=@n2.EmissiveColor) {name=In_Emissive, nodeColor=<rgba 0.1953 0 1 1>}
  n44 = NamedRerouteUsage()
  n45 = Multiply(A=NamedRerouteUsage(), B=ScalarParameter() {defaultValue=1, parameterName=BlurredEdgesMultiplier})
  n46 = NamedRerouteUsage()
  n47 = Multiply(A=Add(A=NamedRerouteUsage(), B=Multiply(A=@n45, B=@n46)), B=Multiply(A=NamedRerouteUsage()) {constB=0.5})
  n48 = NamedRerouteUsage()
  n49 = ComponentMask(None=@n48) {r=true}
  n50 = ComponentMask(None=@n48) {g=true}
  n51 = Saturate(None=Add(A=Add(A=@n49, B=@n50), B=ComponentMask(None=@n48) {b=true}))
  n52 = Multiply(A=Time()) {constB=2}
  n53 = Multiply(A=ConstantBiasScale(None=Sine(None=@n52))) {constB=0.5}
  NamedRerouteDeclaration(None=Multiply(A=LinearInterpolate(A=@n44, B=@n47, Alpha=@n51), B=Add(B=@n53) {constA=1})) {name=BlurredEdgesColorFinal, nodeColor=<rgba 0.125 1 0 1>}
  n54 = NamedRerouteUsage()
  n55 = Add(A=@n54, B=Multiply(A=NamedRerouteUsage(), B=NamedRerouteUsage()))
  n56 = NamedRerouteUsage()
  n57 = NamedRerouteUsage()
  n58 = Add(A=Add(A=ComponentMask(None=@n57) {r=true}, B=ComponentMask(None=@n57) {g=true}), B=ComponentMask(None=@n57) {b=true})
  NamedRerouteDeclaration(None=LinearInterpolate(A=@n54, B=Multiply(A=@n55, B=@n56), Alpha=Saturate(None=@n58))) {name=SharpEdgesColorFinal, nodeColor=<rgba 1 0 0.4453 1>}
  n59 = Add(A=NamedRerouteUsage(), B=NamedRerouteUsage())
  n60 = NamedRerouteUsage()
  n61 = NamedRerouteUsage()
  NamedRerouteDeclaration(None=Multiply(A=Add(A=@n59, B=@n60), B=SmoothStep(Value=@n61) {constMin=0.05})) {name=Out_Emissive, nodeColor=<rgba 0 0.7656 1 1>}
  NamedRerouteDeclaration(None=@n2.BaseColor) {name=In_BaseColor, nodeColor=<rgba 1 0.9141 0 1>}
  n62 = NamedRerouteUsage()
  NamedRerouteDeclaration(None=Add(A=NamedRerouteUsage(), B=Multiply(A=@n62) {constB=0.05})) {name=Out_BaseColor, nodeColor=<rgba 0.5938 0 1 1>}
  n63 = LinearInterpolate(
   A=ScalarParameter() {group=SkyRattle_TechArt, parameterName=RotateNoiseForShoe}
   B=ScalarParameter() {group=SkyRattle_TechArt, parameterName=RotateNoiseForSoleOfShoe}
   Alpha=ComponentMask(None=NamedRerouteUsage()) {b=true}
  ) {constB=0}
  n64 = ScalarParameter() {group=SkyRattle_TechArt, parameterName=RotateNoiseForLeftSideOfShoe}
  n65 = ComponentMask(None=NamedRerouteUsage()) {g=true}
  n66 = NamedRerouteUsage()
  NamedRerouteDeclaration(None=LinearInterpolate(
    A=LinearInterpolate(A=@n63, B=@n64, Alpha=@n65) {constB=0}
    B=ScalarParameter() {group=SkyRattle_TechArt, parameterName=RotateNoiseForRightSideOfShoe}
    Alpha=ComponentMask(None=@n66) {r=true}
   ) {constB=0}) {name=RotationAngle, nodeColor=<rgba 0 0.9062 1 1>}
  NamedRerouteDeclaration(None=MaterialFunctionCall(
    A=VectorParameter() {group=SkyRattle_TechArt, parameterName=RedChannelColorA, defaultValue=<rgba 0.02 0 0.0001 1>}.RGB
    B=VectorParameter() {group=SkyRattle_TechArt, parameterName=RedChannelColorB, defaultValue=<rgba 0.2696 0 0.0017 1>}.RGB
    C=VectorParameter() {group=SkyRattle_TechArt, parameterName=RedChannelColorC, defaultValue=<rgba 0.85 0.0088 0 1>}.RGB
    Alpha=NamedRerouteUsage()
   ) {materialFunction=&Lerp_3Color}.Result) {name=RedChannelColor, nodeColor=<rgba 1 0 0.0241 1>}
  n67 = LinearInterpolate(
   A=AppendVector(
    A=ScalarParameter() {group=SkyRattle_TechArt, parameterName=PivotX_Shoes}
    B=ScalarParameter() {group=SkyRattle_TechArt, parameterName=PivotY_Shoes}
   )
   B=AppendVector(
    A=ScalarParameter() {group=SkyRattle_TechArt, parameterName=PivotX_SoleOfShoes}
    B=ScalarParameter() {group=SkyRattle_TechArt, parameterName=PivotY_SoleOfShoes}
   )
   Alpha=ComponentMask(None=NamedRerouteUsage()) {b=true}
  ) {constB=0}
  n68 = AppendVector(
   A=ScalarParameter() {group=SkyRattle_TechArt, parameterName=PivotX_LeftSideOfShoes}
   B=ScalarParameter() {group=SkyRattle_TechArt, parameterName=PivotY_LeftSideOfShoes}
  )
  n69 = ComponentMask(None=NamedRerouteUsage()) {g=true}
  n70 = ScalarParameter() {group=SkyRattle_TechArt, parameterName=PivotX_RightSideOfShoes}
  n71 = ScalarParameter() {group=SkyRattle_TechArt, parameterName=PivotY_RightSideOfShoes}
  n72 = NamedRerouteUsage()
  NamedRerouteDeclaration(None=LinearInterpolate(A=LinearInterpolate(A=@n67, B=@n68, Alpha=@n69) {constB=0}, B=AppendVector(A=@n70, B=@n71), Alpha=ComponentMask(None=@n72) {r=true}) {constB=0}) {name=PivotLerp, nodeColor=<rgba 1 0.094 0 1>}
  n73 = ScalarParameter() {defaultValue=1, group=SkyRattle_TechArt, parameterName=ScaleX_Shoes}
  n74 = ScalarParameter() {defaultValue=1, group=SkyRattle_TechArt, parameterName=ScaleY_Shoes}
  n75 = ScalarParameter() {defaultValue=1, group=SkyRattle_TechArt, parameterName=ScaleX_SoleOfShoes}
  n76 = ScalarParameter() {defaultValue=1, group=SkyRattle_TechArt, parameterName=ScaleY_SoleOfShoes}
  n77 = NamedRerouteUsage()
  NamedRerouteDeclaration(None=LinearInterpolate(A=AppendVector(A=@n73, B=@n74), B=AppendVector(A=@n75, B=@n76), Alpha=ComponentMask(None=@n77) {b=true}) {constB=0}) {name=ScaleLerp, nodeColor=<rgba 0.4693 1 0 1>}
  n78 = AppendVector(
   A=ScalarParameter() {group=SkyRattle_TechArt, parameterName=RedChannelSpeedX_Shoes}
   B=ScalarParameter() {defaultValue=0.1, group=SkyRattle_TechArt, parameterName=RedChannelSpeedY_Shoes}
  )
  n79 = AppendVector(
   A=ScalarParameter() {group=SkyRattle_TechArt, parameterName=RedChannelSpeedX_SoleOfShoes}
   B=ScalarParameter() {defaultValue=0.1, group=SkyRattle_TechArt, parameterName=RedChannelSpeedY_SoleOfShoes}
  )
  n80 = ComponentMask(None=NamedRerouteUsage()) {b=true}
  NamedRerouteDeclaration(None=Multiply(A=Time(), B=LinearInterpolate(A=@n78, B=@n79, Alpha=@n80) {constB=0})) {name=RedChannelUVLerpAlpha, nodeColor=<rgba 1 0 0.0585 1>}
  n81 = AppendVector(
   A=ScalarParameter() {group=SkyRattle_TechArt, parameterName=GreenChannelSpeedX_Shoes}
   B=ScalarParameter() {defaultValue=0.1, group=SkyRattle_TechArt, parameterName=GreenChannelSpeedY_Shoes}
  )
  n82 = AppendVector(
   A=ScalarParameter() {group=SkyRattle_TechArt, parameterName=GreenChannelSpeedX_SoleOfShoes}
   B=ScalarParameter() {defaultValue=0.1, group=SkyRattle_TechArt, parameterName=GreenChannelSpeedY_SoleOfShoes}
  )
  n83 = ComponentMask(None=NamedRerouteUsage()) {b=true}
  NamedRerouteDeclaration(None=Multiply(A=Time(), B=LinearInterpolate(A=@n81, B=@n82, Alpha=@n83) {constB=0})) {name=GreenChannelUVLerpAlpha, nodeColor=<rgba 0.0055 1 0 1>}
  Saturate(None=@n27)
  Saturate(None=@n42)
 ]}
